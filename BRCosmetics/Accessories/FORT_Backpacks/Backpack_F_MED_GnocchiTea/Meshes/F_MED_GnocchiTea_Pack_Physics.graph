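# Physics asset for the F_MED_GnocchiTea backpack cosmetic, defining simulation bodies and joints for its dangling paper, ropes, grappling hook and toy.
# Capsule bodies branch from backpack_basa through small base bodies into two paper pieces, left and right ropes, and a three-link toy chain.
# The grappling hook hangs off the right rope base, alongside that rope.
# Base joints have free twist and, except the right rope's, free swing, while outer joints use mostly narrow angular limits.
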PhysicsAsset /BRCosmetics/Accessories/FORT_Backpacks/Backpack_F_MED_GnocchiTea/Meshes/F_MED_GnocchiTea_Pack_Physics
bodies {
 backpack_basa [{shapeType=Capsule, length=5, radius=5, shapeName=backpack_basa_capsule, center=<0 -4.0657 0.7193>, rotation=<rot 0 -90 89.9995>}]
 grappling_hook [{
   shapeType=Capsule
   length=15
   radius=1.8
   shapeName=grappling_hook_capsule
   center=<11.6847 -3.2423 -0.5364>
   rotation=<rot 88.3707 -15.2096 0.0041>
  }]
 paper_basa [{shapeType=Capsule, length=0, radius=1, shapeName=paper_basa_capsule}]
 paper_down [{shapeType=Capsule, length=2, radius=2, shapeName=paper_down_capsule, center=<1.8256 0 0>, rotation=<rot 90 180 180>}]
 paper_up [{shapeType=Capsule, length=2, radius=2, shapeName=paper_up_capsule, center=<2 0 0>, rotation=<rot 90 0 0>}]
 rope_basa_l [{shapeType=Capsule, length=0, radius=0.505, shapeName=rope_basa_l_capsule}]
 rope_basa_r [{shapeType=Capsule, length=0, radius=1, shapeName=rope_basa_r_capsule}]
 rope_l [{shapeType=Capsule, length=5, radius=3, shapeName=rope_l_capsule, center=<4.2352 0 0>, rotation=<rot 90 0 -0.0001>}]
 rope_r [{shapeType=Capsule, length=10, radius=1, shapeName=rope_r_capsule, center=<5.1665 0.4738 -0.245>, rotation=<rot 6.9997 179.9994 134.9991>}]
 toy_01 [{shapeType=Capsule, length=0.5, radius=2, shapeName=toy_01_capsule, center=<1.1333 0 0>, rotation=<rot 90 0 0>}]
 toy_02 [{shapeType=Capsule, length=0, radius=2, shapeName=toy_02_capsule, center=<0.9998 0 0>, rotation=<rot 90 0 0>}]
 toy_03 [{shapeType=Capsule, length=4, radius=2, shapeName=toy_03_capsule, center=<3.7358 0 0>, rotation=<rot 90 0 0>}]
 toy_basa [{shapeType=Capsule, length=0, radius=0.505, shapeName=toy_basa_capsule}]
}
constraints [
 {
  bone1Name=paper_basa
  bone2Name=backpack_basa
  swing1LimitDegrees=45
  swing1Motion=Free
  swing2LimitDegrees=45
  swing2Motion=Free
  twistLimitDegrees=45
  twistMotion=Free
 }
 {bone1Name=paper_down, bone2Name=paper_basa, swing1LimitDegrees=5, swing2LimitDegrees=5, twistLimitDegrees=3}
 {bone1Name=paper_up, bone2Name=paper_basa, swing1LimitDegrees=6, swing2LimitDegrees=8, twistLimitDegrees=3}
 {
  bone1Name=rope_basa_l
  bone2Name=backpack_basa
  swing1LimitDegrees=45
  swing1Motion=Free
  swing2LimitDegrees=45
  swing2Motion=Free
  twistLimitDegrees=45
  twistMotion=Free
 }
 {bone1Name=rope_l, bone2Name=rope_basa_l, swing1LimitDegrees=15, swing2LimitDegrees=15, twistLimitDegrees=2}
 {bone1Name=rope_basa_r, bone2Name=backpack_basa, swing1LimitDegrees=45, swing2LimitDegrees=45, twistLimitDegrees=45, twistMotion=Free}
 {bone1Name=rope_r, bone2Name=rope_basa_r, swing1LimitDegrees=15, swing2LimitDegrees=2, twistLimitDegrees=2}
 {
  bone1Name=toy_basa
  bone2Name=backpack_basa
  swing1LimitDegrees=45
  swing1Motion=Free
  swing2LimitDegrees=45
  swing2Motion=Free
  twistLimitDegrees=45
  twistMotion=Free
 }
 {bone1Name=toy_01, bone2Name=toy_basa, swing1LimitDegrees=15, swing2LimitDegrees=60, twistLimitDegrees=1}
 {bone1Name=toy_02, bone2Name=toy_01, swing1LimitDegrees=20, swing2LimitDegrees=20, twistLimitDegrees=5}
 {bone1Name=grappling_hook, bone2Name=rope_basa_r, swing1LimitDegrees=25, swing2LimitDegrees=10, twistLimitDegrees=2}
 {bone1Name=toy_03, bone2Name=toy_02, swing1LimitDegrees=20, swing2LimitDegrees=20, twistLimitDegrees=5}
]
