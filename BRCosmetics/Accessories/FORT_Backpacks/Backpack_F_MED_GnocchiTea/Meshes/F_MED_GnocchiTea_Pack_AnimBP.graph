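# Animation Blueprint for the GnocchiTea backpack cosmetic, adding secondary physics motion and a tilt that responds to the wearer's spine bend.
# A RigidBody node simulates the input pose relative to backpack_basa, with gravity from a deprecated gravity-override function updated each frame.
# An additive rotation and offset on the pack bone then blends in by BackpackRotate, mapped from spine_05's yaw delta against spine_01.
# A second RigidBody node is left unconnected.
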AnimBlueprint /BRCosmetics/Accessories/FORT_Backpacks/Backpack_F_MED_GnocchiTea/Meshes/F_MED_GnocchiTea_Pack_AnimBP
target_skeleton=&/F_MED_GnocchiTea_Pack_Skeleton
graphs [
 AnimGraph {nodes [
   RigidBody() {
    bOverrideWorldGravity=true
    lODThreshold=2
    overridePhysicsAsset=&/F_MED_GnocchiTea_Pack_Physics
    simulationSpace=BaseBoneSpace
    alphaBoolBlend {blendOption=Linear}
    baseBoneRef {boneName=backpack_basa}
    componentAppliedLinearAccClamp=<3000 3000 3000>
    componentLinearAccScale=<0.4 0.5 0.4>
    componentLinearVelScale=<0.4 0.5 0.4>
    overrideWorldGravity {z=-980}
    simSpaceSettings {maxAngularAcceleration=250, maxAngularVelocity=3, maxLinearAcceleration=250, maxLinearVelocity=250, velocityScaleZ=0, worldAlpha=1}
   }
   n1 = RigidBody!(ComponentPose=Local To Component(LocalPose=Input Pose()), Alpha, OverrideWorldGravity:Vector=Get Gravity Override(), ExternalForce) -> (Pose) {
    bOverrideWorldGravity=true
    overridePhysicsAsset=&F_MED_GnocchiTea_Pack_Physics
    simulationSpace=BaseBoneSpace
    alphaBoolBlend {blendOption=Linear}
    baseBoneRef {boneName=backpack_basa}
    componentAppliedLinearAccClamp=<3000 3000 3000>
    componentLinearAccScale=<0.5 0.5 0.5>
    componentLinearVelScale=<0.4 0.4 0.4>
    overrideWorldGravity {z=-980}
    simSpaceSettings {maxAngularAcceleration=250, maxAngularVelocity=3, maxLinearAcceleration=250, maxLinearVelocity=250, velocityScaleZ=0, worldAlpha=1}
   }
   n2 = "Transform (Modify) Bone - Bone: pack"(ComponentPose=@n1, Alpha=Get BackpackRotate() -> (BackpackRotate:double), Translation="0,-10,0", Rotation="0,0,74") {
    rotationMode=BMM_Additive
    rotationSpace=BCS_BoneSpace
    scaleMode=BMM_Additive
    scaleSpace=BCS_BoneSpace
    translationMode=BMM_Additive
    translationSpace=BCS_BoneSpace
    alphaBoolBlend {blendOption=Linear}
    boneToModify {boneName=pack}
    rotation {roll=74}
    translation {y=-10}
   }
   Output Pose(Result=Component To Local(ComponentPose=@n2)) {layerGroup=None}
  ]}
 EventGraph {nodes [
   Event BlueprintUpdateAnimation()
   > n1 = Sequence()
   > n2 = Gravity Override State Function DEPRECATED(CharacterPartAnimInstance=Self-Reference(), GameTimeInSeconds=GetGameTimeInSeconds(), GravityOverrideParams=Get Gravity Override Params())
   > Set Gravity Override(Gravity Override=@n2)
   n3 = BreakTransform(InTransform=GetDeltaTransformFromRefPose(self=GetOwnerSkeletalMeshComponent(), BoneName=spine_05, BaseName=spine_01))
   Set BackpackRotate(execute=@n1.then_1, BackpackRotate:double=MapRangeClamped(Value=BreakRotator(InRot=@n3.Rotation).Yaw, InRangeA=-25, InRangeB=-65)) -> (Output_Get:double)
   TryGetPawnOwner()
   Reroute Node(InputPin:Rotator=@n3.Rotation) -> (OutputPin:Rotator)
  ]}
]
variables [
 BackpackRotate {type=number}
 Gravity Override {type=Vector}
 Gravity Override Params {type=GravityOverrideParamsStruct}
]
